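# Niagara trail effect for the SliceVine v2 glider, with gameplay and front-end preview tuning, variant colors and time-of-day response.
# Sidewings draws eroding anim-trail ribbons from skeletal mesh bone pairs, while Speedlines spawns speed-stretched torus sprites pushed by the ForwardDot and RightDot inputs.
# Mirrored WingTrail emitters place burst-spawned ribbon points from a socket along a slerped trail vector, lengthened by a ForwardDot-driven timeline.
# Their steering-based sideways offset is multiplied by zero.
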NiagaraSystem /BRCosmetics/Gadgets/Assets/VinderTech_GliderChute/Glider_SliceVine/FX/NS_Glider_SliceVine_v2_Trail
dependencies {
 usedDataInterfaces [
  &NiagaraDataInterfaceCurve
  &NiagaraDataInterfaceVector2DCurve
  &NiagaraDataInterfaceColorCurve
  &NiagaraDataInterfaceSkeletalMesh
 ]
 usedDynamicInputs [&Boolean_AndOperation]
 usedModules [
  &self:SystemSpawnScript.SystemScriptSource.SystemScriptGraph.NiagaraNodeAssignment_0.SetVariables_#1
  &SystemState
  &FN_VariantManager
  /BRCosmetics/Characters/Player/Female/Medium/Bodies/F_MED_PiperShelf_Pearl/FX/TimeOfDayGlobal
 ]
}
emitters {
 Sidewings {
  emitterUpdateScript [
   &EmitterState()
   &FN_SpawnAnimTrail_v2(Number of Anim Trails=2)
  ]
  particleSpawnScript [
   &PartitionParticles(ID to Partition:NiagaraInt32=$Engine.ExecIndex)
   /Game/Effects/Niagara/Modules/AnimTrail/FN_InitializeAnimTrail/FN_InitializeAnimTrailBoneIndecies(Second Bone or Socket Index=1)
   &FN_InitializeAnimTrailBoneIndecies(AnimTrail Selection=1, First Bone or Socket Index=2, Second Bone or Socket Index=3) {module=FN_InitializeAnimTrailBoneIndecies001}
   &FN_InitializeAnimTrail(Lifetime=0.4, SkeletalMesh=$System.Skeletal Mesh)
   &AddVelocity(Velocity=&RotateVectorByQuaternion(
     Vector To Rotate=&MakeVector(X=&FN_FrontEndFloatModifier(Gameplay=650, Front End=-650), Y=0, Z=0)
     Quaternion=<0 0 1 1>
    ))
  ]
  particleUpdateScript [
   &InterpolateOverTime(Initial Float Value:NiagaraFloat=$System.Timeline.Time, Rate Of Change=0.1, Target Float Value=1, Use Set Interpolation Rate=true)
   &KillParticles(Kill Particles=&SetBoolByFloatComparison(A=$System.Timeline.Time, B=$Particles.InterpolateOverTime.InterpolateOverTime_Float.Moving Average, Comparison Type=::NewEnumerator6))
   &FN_UpdateAnimTrail(SkeletalMesh=$System.Skeletal Mesh)
   &FN_UpdateAnimTrail(SkeletalMesh=$System.Skeletal Mesh) {module=FN_UpdateAnimTrail001}
   &InheritSourceMovement(
    Applied Position Scale=&Multiply_VectorByFloat(
     Vector=<1 1 1>
     Float=&FloatFromCurve(FloatCurve=DataInterface() {propertyValues {Curve {keys [
          {interpMode=RCIM_Cubic, tangentMode=RCTM_User, time=0.66, value=1}
          {arriveTangent=-1.9042, interpMode=RCIM_Cubic, leaveTangent=-1.9042, tangentMode=RCTM_User, time=1, value=0.5}
         ]}}})
    )
    ApplyForce=false
    ApplyPosition=true
    ApplyPositionFalloff=false
   )
   &SolveForcesAndVelocity(Clamp Velocity=false, Limit Acceleration=false)
   /SetVariables(Particles.Position=&AddVectorToPosition(
     Position=$Particles.Position
     Vector=&TransformVector(
      Vector=&FN_FrontEndVector3Modifier(Gameplay=<0 -6 -1.5>, Front End=<0 6 -1.5>)
      Destination Space=::Simulation
     )
    )) {isSetParameters=true}
   &FN_RibbonMaterial(
    Color Scale U=0.2182
    Erosion Amount:NiagaraFloat=&FloatFromCurve(FloatCurve=DataInterface() {propertyValues {Curve {keys [{tangentMode=RCTM_User, time=0, value=1}, {tangentMode=RCTM_User, time=1, value=0}]}}})
    Erosion Pan Offset U:NiagaraFloat=$Particles.MaterialRandom
    Erosion Pan Speed U=0.3
    Erosion Power=2
    Erossion Scale U=0.1
   )
   &ScaleColor(
    Scale Alpha=&FloatFromCurve(
     FloatCurve=DataInterface() {propertyValues {Curve {keys [<at 0 1 RCIM_Cubic>,<at 1 0>]}}}
     Scale Curve=&FN_FrontEndFloatModifier(Gameplay=$System.Timeline.Time, Front End=0.6)
    )
    Scale RGB=&MakeVectorFromLinearColorRGB(Color=&VariantColorSelector(
      Variant Amount=0
      BlendAmount=0.5
      Use TODCurve?=true
      Original Color=<rgba 0.2091 0.3981 0.8906 1>
      Variant 1 Color=<rgba 0.4641 0.0967 0.5156 1>
      Variant 2 Color=?
      Front End Color=&MakeCustomLinearColorFromBool(Bool=$System.bIsVariant0, True Color=<rgba 0.2091 0.3981 0.8906 1>, False Color=<rgba 0.3247 0.1235 0.5156 1>)
      Gameplay Color=&MakeCustomLinearColorFromBool(Bool=$System.bIsVariant0, True Color=<rgba 0.2091 0.3981 0.8906 1>, False Color=<rgba 0.3247 0.1235 0.5156 1>)
     ))
    ScaleA=true
    ScaleRGB=true
   )
   /SetVariables(Particles.RibbonU0Override=$Particles.NormalizedAge) {isSetParameters=true}
  ]
  rendererClasses [&NiagaraRibbonRendererProperties]
  renderers [{rendererClass=&NiagaraRibbonRendererProperties}]
 }
 Speedlines {
  emitterUpdateScript [
   &EmitterState()
   &SpawnRate(
    Spawn Probability=&MakeCustomFloatFromBool(Bool=&SetBoolByFloatComparison(A=$System.Timeline.Time, B=0), True Float=1, False Float=0)
    SpawnRate=&MakeCustomFloatFromBool(Bool=$User.bIsFrontEndPreview, True Float=50, False Float=90)
    Use Spawn Probability=true
   ) {module=SpawnRate001}
  ]
  particleSpawnScript [
   &InitializeParticle(
    Color=&MakeLinearColorFromVectorAndFloat(
     "Vector (RGB)"=<1 1 1>
     "Float (Alpha)"=&Multiply_Float(
      A=&UniformRangedFloat(Minimum=1, Maximum=1)
      B=&MakeCustomFloatFromBool(
       Bool=$User.bIsFrontEndPreview
       True Float=&FloatFromCurve(
        FloatCurve=DataInterface() {propertyValues {Curve {keys [{tangentMode=RCTM_User, time=0, value=0}, {tangentMode=RCTM_User, time=2, value=1}]}}}
        CurveIndex=$Emitter.Age
        Scale Curve=0.4
       )
       False Float=1
      )
     )
    )
    Lifetime Max=0.2
    Lifetime Min=0.15
    Lifetime Mode=::NewEnumerator1
    Position Offset=&FN_FrontEndVector3Modifier(Gameplay=<0 -50 0>, Front End=<0 50 0>)
    Sprite Size Max=<3 10>
    Sprite Size Min=<2 2>
    Sprite Size Mode=::NewEnumerator2
    UsePositionOffset=true
    Write Material Random=true
   )
   &AddVelocity(
    Module.Use Falloff Distance=false
    Override Seed=false
    Velocity Mode=::NewEnumerator1
    Velocity Speed=&FN_FrontEndFloatModifier(Gameplay=-2500, Front End=-2500)
   ) {module=AddVelocity001}
   &TorusLocation(
    Align Torus With Axis=true
    Coordinate Space:ENiagaraCoordinateSpace=::Simulation
    Handle Radius=0
    Large Radius:NiagaraFloat=&FN_FrontEndFloatModifier(
     Gameplay=&RandomRangeFloat(Minimum=450, Maximum=500)
     Front End=&RandomRangeFloat(Minimum=350, Maximum=400)
    )
    MaskSpawnGroup=false
    Module.Override Randomness=false
    Offset=<0 33 0>
    Torus Axis=<0 1 0>
    Torus Origin:NiagaraPosition=&AddVectorToPosition(
     Position=$Particles.Position
     Vector=&FN_FrontEndVector3Modifier(Gameplay=<0 -500 0>, Front End=<0 250 0>)
    )
   )
   &AddVelocity(
    Velocity=&Add_Vector(
     A=&RandomRangeVector(Minimum=<0 300 0>, Maximum=<0 700 1>)
     B=&Add_Vector(
      A=&Multiply_VectorByFloat(Vector=<-555 0 0>, Float=$User.RightDot)
      B=&Multiply_VectorByFloat(Vector=<0 0 222>, Float=$User.ForwardDot)
     )
    )
    Velocity Speed Scale=&FN_FrontEndFloatModifier(Gameplay=1, Front End=1)
   )
  ]
  particleUpdateScript [
   &SolveForcesAndVelocity(Clamp Velocity=false, Limit Acceleration=false)
   &ParticleState()
   &ScaleSpriteSize(Scale Factor=&Vector2DFromCurve(
     Vector2Curve=DataInterface() {propertyValues {XCurve {keys [<at 0 0.5 RCIM_Cubic>,<at 0.5 1 RCIM_Cubic>]}, YCurve {keys [<at 0 0 RCIM_Cubic>,<at 0.5 1 RCIM_Cubic>]}}}
     ScaleCurve=&Vector2DFromFloat(Value=1)
    ))
   &ScaleSpriteSizeBySpeed(Max Scale Factor=<1 50>, Min Scale Factor=<0.01 0.05>, Velocity Threshold=1)
   &ScaleColor(
    Scale Alpha=&FloatFromCurve(
     FloatCurve=DataInterface() {propertyValues {Curve {keys [
         <at 0 0 RCIM_Cubic>
         <at 0.3759 1.0207>
         {arriveTangent=-2.221, interpMode=RCIM_Cubic, leaveTangent=-2.221, tangentMode=RCTM_User, time=1, value=0}
        ]}}}
     Scale Curve=&MakeCustomFloatFromBool(
      Bool=$User.bIsFrontEndPreview
      True Float=&MakeCustomFloatFromBool(Bool=$User.bIsFullyDeployed, True Float=0.5, False Float=0)
      False Float=&Lerp_Float(A=0, B=0.5, Alpha=&FN_FrontEndFloatModifier(Gameplay=$System.Timeline.Time, Front End=0.5))
     )
    )
    Scale RGB=&VectorFromFloat(Value=&FN_FrontEndFloatModifier(Gameplay=0.5, Front End=0.33))
    ScaleA=true
    ScaleRGB=true
   )
  ]
  rendererClasses [&NiagaraSpriteRendererProperties]
  renderers [{rendererClass=&NiagaraSpriteRendererProperties}]
 }
 WingTrail_L {
  emitterUpdateScript [
   &EmitterState(
    Inactive Response=::NewEnumerator1
    Life Cycle Mode=::NewEnumerator1
    Loop Behavior=::NewEnumerator1
    Loop Duration Mode=::NewEnumerator1
    UseLoopDelay=false
   )
   &Timeline(Play=&SetBoolByFloatComparison(A=$User.ForwardDot, B=0.1))
   &InterpolateOverTime(Target Float Value:NiagaraFloat=$User.RightDot)
   &SampleSkeletalMesh(
    AdditionalYawPitchRoll=false
    Bone / Socket Index=4
    Mesh Sampling Type:ENiagaraSkelSamplingModeFull=::NewEnumerator5
    Skeletal Mesh:NiagaraDataInterfaceSkeletalMesh=$System.Skeletal Mesh
    Socket Sampling Mode:ENiagaraSocketSamplingMode=::NewEnumerator5
   )
   /SetVariables(
    Emitter.PositionOffset:Vector3f=&TransformVector(Vector=$StackContext.SampleSkeletalMesh.SampledPosition, Destination Space=::Simulation)
    Emitter.RibbonColor:LinearColor=&VariantColorSelector(
     Variant Amount=0
     Morning Color=<rgba 1 1 1 0.3>
     Day Color=<rgba 1 1 1 0.2>
     Evening Color=<rgba 1 1 1 0.3>
     Night Color=<rgba 1 1 1 0.5>
     Use TOD?=true
     Original Color=<rgba 1 0.3813 0.0262 0.5>
     Variant 1 Color=<rgba 0.2159 0.8963 0.2232 0.5>
     Variant 2 Color=?
     Front End Color=&MakeCustomLinearColorFromBool(Bool=$System.bIsVariant0, True Color=<rgba 0.2091 0.3981 0.8906 1>, False Color=<rgba 0.3247 0.1235 0.5156 1>)
     Gameplay Color=&MakeCustomLinearColorFromBool(Bool=$System.bIsVariant0, True Color=<rgba 0.2091 0.3981 0.8906 1>, False Color=<rgba 0.3247 0.1235 0.5156 1>)
    )
    Emitter.RibbonLength:NiagaraFloat=&FN_FrontEndFloatModifier(Gameplay=&Lerp_Float(A=30, B=100, Alpha=$StackContext.Timeline.Time), Front End=-100)
   ) {isSetParameters=true, module=SetVariables_#2}
   &SpawnBurst_Instantaneous(Spawn Count=4, Use Spawn Probability=false, UseLoopCountLimit=false)
   /BRCosmetics/Animation/Game/MainPlayer/Emotes/UpbeatIguana/FX/InterpolateTrailVector(Range=0.3, Target Vector:Vector3f=$Engine.Owner.SystemYAxis)
  ]
  particleSpawnScript [/SetVariables(Particles.RibbonWidth=&Multiply_Float(A=5, B=&Lerp_Float(A=0.66, B=1, Alpha=&ReturnNormalizedExecIndex()))) {isSetParameters=true, module=SetVariables_#3}]
  particleUpdateScript [
   /SetVariables(Particles.Color=&MultiplyLinearColor(
     Linear Color=$Emitter.RibbonColor
     Scale Factor=&MakeVector4_FromFloat(Float=&Multiply_Float(
       A=$System.Timeline.Time
       B=&FNTimeOfDayScalar(Night Scale=0.5, Morning Scale=0.8, Evening Scale=0.8)
      ))
    )) {isSetParameters=true, module=SetVariables_#4}
   /SetVariables(
    Particles.InterpolatedTrailVector:Vector3f=&TransformVector(
     Vector=/Niagara/DynamicInputs/Math/SlerpVector(
      Vector A:Vector3f=$Emitter.InterpolatedVector
      Vector B:Vector3f=$Emitter.TargetVector
      Alpha:NiagaraFloat=&ReturnNormalizedExecIndex()
      Scale:NiagaraFloat=1
     )
     Source Space=::World
     Destination Space=::Simulation
     Apply Scale=true
    )
    Particles.RibbonFacing=&TransformVector(Vector=$Engine.Owner.SystemZAxis, Source Space=::World, Destination Space=::Simulation)
   ) {isSetParameters=true, module=SetVariables_#5}
   /SetVariables(Particles.SidewaysOffsetVector:Vector3f=&Multiply_VectorByFloat(
     Vector=&NormalizeVector(
      Vector To Normalize=/Niagara/DynamicInputs/Vectors/CrossProduct(
       Vector A:Vector3f=$StackContext.InterpolatedTrailVector
       Vector B:Vector3f=$Particles.RibbonFacing
       Normalize Incoming Vectors:NiagaraBool=true
      )
      Safe Normalize=true
      Safe Fallback Threshold=0
      Safe Fallback Vector=<0 0 0.001>
     )
     Float=&Multiply_Float(
      A=$Emitter.InterpolateOverTime.InterpolateOverTime_Float.Moving Average
      B=&Multiply_Float(
       A=&Power(a=&OneMinusFloat(Float=&ReturnNormalizedExecIndex()), b=2)
       B=0
      )
     )
    )) {isSetParameters=true, module=SetVariables_#6}
   /SetVariables(Particles.InterpolatedPosition:NiagaraPosition=&AddVectorToPosition(
     Position=&AddVectorToPosition(
      Position=$Engine.Emitter.SimulationPosition
      Vector=&Add_Vector(A=$Emitter.PositionOffset, B=$StackContext.SidewaysOffsetVector)
     )
     Vector=&Multiply_VectorByFloat(
      Vector=$Particles.InterpolatedTrailVector
      Float=&Multiply_Float(A=&OneMinusFloat(Float=&ReturnNormalizedExecIndex()), B=$Emitter.RibbonLength)
     )
    )) {isSetParameters=true, module=SetVariables_#7}
  ]
  rendererClasses [&NiagaraRibbonRendererProperties]
  renderers [{rendererClass=&NiagaraRibbonRendererProperties}]
 }
 WingTrail_R {
  emitterUpdateScript [
   &EmitterState(
    Inactive Response=::NewEnumerator1
    Life Cycle Mode=::NewEnumerator1
    Loop Behavior=::NewEnumerator1
    Loop Duration Mode=::NewEnumerator1
    UseLoopDelay=false
   )
   &Timeline(Play=&SetBoolByFloatComparison(A=$User.ForwardDot, B=0.1))
   &InterpolateOverTime(Target Float Value:NiagaraFloat=$User.RightDot)
   &SampleSkeletalMesh(
    AdditionalYawPitchRoll=false
    Bone / Socket Index=5
    Mesh Sampling Type:ENiagaraSkelSamplingModeFull=::NewEnumerator5
    Skeletal Mesh:NiagaraDataInterfaceSkeletalMesh=$System.Skeletal Mesh
    Socket Sampling Mode:ENiagaraSocketSamplingMode=::NewEnumerator5
   )
   /SetVariables(
    Emitter.PositionOffset:Vector3f=&TransformVector(Vector=$StackContext.SampleSkeletalMesh.SampledPosition, Destination Space=::Simulation)
    Emitter.RibbonColor:LinearColor=&VariantColorSelector(
     Variant Amount=0
     Morning Color=<rgba 1 1 1 0.3>
     Day Color=<rgba 1 1 1 0.2>
     Evening Color=<rgba 1 1 1 0.3>
     Night Color=<rgba 1 1 1 0.5>
     Use TOD?=true
     Original Color=<rgba 1 0.3813 0.0262 0.5>
     Variant 1 Color=<rgba 0.2159 0.8963 0.2232 0.5>
     Variant 2 Color=?
     Front End Color=&MakeCustomLinearColorFromBool(Bool=$System.bIsVariant0, True Color=<rgba 0.2091 0.3981 0.8906 1>, False Color=<rgba 0.3247 0.1235 0.5156 1>)
     Gameplay Color=&MakeCustomLinearColorFromBool(Bool=$System.bIsVariant0, True Color=<rgba 0.2091 0.3981 0.8906 1>, False Color=<rgba 0.3247 0.1235 0.5156 1>)
    )
    Emitter.RibbonLength:NiagaraFloat=&FN_FrontEndFloatModifier(Gameplay=&Lerp_Float(A=30, B=100, Alpha=$StackContext.Timeline.Time), Front End=-100)
   ) {isSetParameters=true, module=SetVariables_#2}
   &SpawnBurst_Instantaneous(Spawn Count=4, Use Spawn Probability=false, UseLoopCountLimit=false)
   &InterpolateTrailVector(Range=0.3, Target Vector:Vector3f=$Engine.Owner.SystemYAxis)
  ]
  particleSpawnScript [/SetVariables(Particles.RibbonWidth=&Multiply_Float(A=5, B=&Lerp_Float(A=0.66, B=1, Alpha=&ReturnNormalizedExecIndex()))) {isSetParameters=true, module=SetVariables_#3}]
  particleUpdateScript [
   /SetVariables(Particles.Color=&MultiplyLinearColor(
     Linear Color=$Emitter.RibbonColor
     Scale Factor=&MakeVector4_FromFloat(Float=&Multiply_Float(
       A=$System.Timeline.Time
       B=&FNTimeOfDayScalar(Night Scale=0.5, Morning Scale=0.8, Evening Scale=0.8)
      ))
    )) {isSetParameters=true, module=SetVariables_#4}
   /SetVariables(
    Particles.InterpolatedTrailVector:Vector3f=&TransformVector(
     Vector=&SlerpVector(
      Vector A:Vector3f=$Emitter.InterpolatedVector
      Vector B:Vector3f=$Emitter.TargetVector
      Alpha:NiagaraFloat=&ReturnNormalizedExecIndex()
      Scale:NiagaraFloat=1
     )
     Source Space=::World
     Destination Space=::Simulation
     Apply Scale=true
    )
    Particles.RibbonFacing=&TransformVector(Vector=$Engine.Owner.SystemZAxis, Source Space=::World, Destination Space=::Simulation)
   ) {isSetParameters=true, module=SetVariables_#5}
   /SetVariables(Particles.SidewaysOffsetVector:Vector3f=&Multiply_VectorByFloat(
     Vector=&NormalizeVector(
      Vector To Normalize=&CrossProduct(
       Vector A:Vector3f=$StackContext.InterpolatedTrailVector
       Vector B:Vector3f=$Particles.RibbonFacing
       Normalize Incoming Vectors:NiagaraBool=true
      )
      Safe Normalize=true
      Safe Fallback Threshold=0
      Safe Fallback Vector=<0 0 0.001>
     )
     Float=&Multiply_Float(
      A=$Emitter.InterpolateOverTime.InterpolateOverTime_Float.Moving Average
      B=&Multiply_Float(
       A=&Power(a=&OneMinusFloat(Float=&ReturnNormalizedExecIndex()), b=2)
       B=0
      )
     )
    )) {isSetParameters=true, module=SetVariables_#6}
   /SetVariables(Particles.InterpolatedPosition:NiagaraPosition=&AddVectorToPosition(
     Position=&AddVectorToPosition(
      Position=$Engine.Emitter.SimulationPosition
      Vector=&Add_Vector(A=$Emitter.PositionOffset, B=$StackContext.SidewaysOffsetVector)
     )
     Vector=&Multiply_VectorByFloat(
      Vector=$Particles.InterpolatedTrailVector
      Float=&Multiply_Float(A=&OneMinusFloat(Float=&ReturnNormalizedExecIndex()), B=$Emitter.RibbonLength)
     )
    )) {isSetParameters=true, module=SetVariables_#7}
  ]
  rendererClasses [&NiagaraRibbonRendererProperties]
  renderers [{rendererClass=&NiagaraRibbonRendererProperties}]
 }
}
summary {
 systemName=NS_Glider_SliceVine_v2_Trail
 emitters [
  {emitterName=Sidewings, rendererClasses [&NiagaraRibbonRendererProperties]}
  {emitterName=Speedlines, rendererClasses [&NiagaraSpriteRendererProperties]}
  {emitterName=WingTrail_L, rendererClasses [&NiagaraRibbonRendererProperties]}
  {emitterName=WingTrail_R, rendererClasses [&NiagaraRibbonRendererProperties]}
 ]
 userVariables [
  User.bIsFrontEnd {type=NiagaraBool}
  User.bIsFrontEndPreview {type=NiagaraBool}
  User.bIsFullyDeployed {type=NiagaraBool}
  User.ForwardDot {type=NiagaraFloat, defaultValue=0, description=""}
  User.RibbonIDOffset {type=NiagaraInt32}
  User.RightDot {type=NiagaraFloat, defaultValue=0, description=""}
  User.Variant {type=NiagaraFloat, defaultValue=0, description=""}
 ]
}
