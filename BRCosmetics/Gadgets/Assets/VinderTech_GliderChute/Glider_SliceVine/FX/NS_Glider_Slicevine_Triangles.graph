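# Niagara effect for the SliceVine glider that emits a brief burst of short-lived mesh triangles, with spawn probability fading late in the emitter loop.
# Triangles spawn in a cylinder, and a direct position write pulls them inward along their initial normals while the ring sweeps along Y with system age.
# Curves over particle life pop the scale in and out and flash brightness.
# A VariantManager selector sets an emitter color-variant float.
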NiagaraSystem /BRCosmetics/Gadgets/Assets/VinderTech_GliderChute/Glider_SliceVine/FX/NS_Glider_Slicevine_Triangles
dependencies {
 usedDataInterfaces [&NiagaraDataInterfaceCurve]
 usedDynamicInputs []
 usedModules [&SystemState, &FN_VariantManager]
}
emitters {Triangles {
  emitterUpdateScript [
   &EmitterState()
   &SpawnRate(
    Spawn Probability=&FloatFromCurve(
     FloatCurve=DataInterface() {propertyValues {Curve {keys [{interpMode=RCIM_Cubic, tangentMode=RCTM_User, time=0.5, value=1}, {interpMode=RCIM_Cubic, tangentMode=RCTM_User, time=1, value=0}]}}}
     CurveIndex=$Emitter.NormalizedLoopAge
    )
    SpawnRate=30
    Use Spawn Probability=true
   ) {module=SpawnRate001}
   /SetVariables(Emitter.ColoVarrFloat:NiagaraFloat=/BRCosmetics/Weapons/FORT_Melee/Pickaxe_GateHound/FX/VariantManager/VariantFloatSelector(
     Original Float:NiagaraFloat=0.5
     Variant 1 Float:NiagaraFloat=0.1
     Variant 2 Float:NiagaraFloat=0.075
     Variant 3 Float:NiagaraFloat=?
     Variant Amount:NiagaraInt32=1
    )) {isSetParameters=true}
  ]
  particleSpawnScript [
   &InitializeParticle(
    Color Mode=::NewEnumerator0
    Lifetime=0.3
    Mesh Scale Max=<0.6 0.6 0.6>
    Mesh Scale Min=<1.2 1.2 1.2>
    Mesh Scale Mode=::NewEnumerator2
    UsePositionOffset=false
    Write Material Random=true
   )
   &InitialMeshOrientation(Use Rotation Vector=false)
   &ShapeLocation(
    Cylinder Height=5
    Cylinder Radius=150
    MaskSpawnGroup=false
    Override Seed=false
    Rotation Axis=<0 1 0>
    Rotation Mode=::NewEnumerator1
    Shape Primitive=::NewEnumerator1
    Surface Only=false
    UseOwnerScale=true
   )
   &AddVelocity(Velocity=$Particles.ShapeLocation.ShapeNormal, Velocity Speed Scale=-100)
   &ApplyInitialForces()
  ]
  particleUpdateScript [
   &ParticleState()
   /SetVariables(Particles.Position=&AddVectorToPosition(
     Position=&Multiply_VectorByFloat(
      Vector=$Particles.Initial.ShapeLocation.ShapeNormal
      Float=&Lerp_Float(
       A=150
       B=20
       Alpha=&FloatFromCurve(FloatCurve=DataInterface() {propertyValues {Curve {keys [{tangentMode=RCTM_User, time=0, value=0}, {tangentMode=RCTM_User, time=1, value=1}]}}})
      )
     )
     Vector=&TransformVector(
      Vector=&Lerp_Vector(
       A=<0 125 0>
       B=<0 -150 0>
       Alpha=&FloatFromCurve(
        FloatCurve=DataInterface() {propertyValues {Curve {keys [{tangentMode=RCTM_User, time=0, value=1}, {tangentMode=RCTM_User, time=1, value=0}]}}}
        CurveIndex=$Engine.System.Age
       )
      )
      Destination Space=::Simulation
     )
    )) {isSetParameters=true}
   &ScaleMeshSize(Scale Factor=&Multiply_VectorByFloat(
     Vector=<1 1 1>
     Float=&FloatFromCurve(FloatCurve=DataInterface() {propertyValues {Curve {keys [
          {arriveTangent=9.0725, interpMode=RCIM_Cubic, leaveTangent=9.0725, tangentMode=RCTM_User, time=0, value=0}
          {arriveTangent=0.0659, interpMode=RCIM_Cubic, leaveTangent=0.0659, tangentMode=RCTM_User, time=0.1408, value=1}
          {arriveTangent=0.1312, interpMode=RCIM_Cubic, leaveTangent=0.1312, tangentMode=RCTM_User, time=0.8, value=1}
          {arriveTangent=-0.0405, interpMode=RCIM_Cubic, leaveTangent=-0.0405, tangentMode=RCTM_User, time=1, value=0}
         ]}}})
    ))
   &UpdateMeshOrientation()
   &ScaleColor(
    Scale RGB=&VectorFromFloat(Value=&FloatFromCurve(FloatCurve=DataInterface() {propertyValues {Curve {keys [
          {arriveTangent=9.0725, interpMode=RCIM_Cubic, leaveTangent=9.0725, tangentMode=RCTM_User, time=0, value=1}
          {arriveTangent=-2.5249, interpMode=RCIM_Cubic, leaveTangent=-2.5249, tangentMode=RCTM_User, time=0.1408, value=20}
          {arriveTangent=-21.6542, interpMode=RCIM_Cubic, leaveTangent=-21.6542, tangentMode=RCTM_User, time=0.3, value=5}
          {arriveTangent=-0.0405, interpMode=RCIM_Cubic, leaveTangent=-0.0405, tangentMode=RCTM_User, time=0.75, value=0.25}
         ]}}}))
    ScaleA=true
    ScaleRGB=true
   )
   &SolveForcesAndVelocity(Clamp Velocity=false, Limit Acceleration=false)
  ]
  rendererClasses [&NiagaraMeshRendererProperties]
  renderers [{rendererClass=&NiagaraMeshRendererProperties, rendererIndex=0}]
 }}
summary {
 systemName=NS_Glider_Slicevine_Triangles
 emitters [{emitterName=Triangles, rendererClasses [&NiagaraMeshRendererProperties]}]
 userVariables [User.Variant {type=NiagaraFloat, defaultValue=2, description=""}]
}
